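; Ancient mask prop Blueprint that picks a texture swatch for its paper mesh at construction.
; When randomize-by-position is enabled, it derives the swatch index from averaged absolute world coordinates, scaled, rounded and wrapped to the array length, giving deterministic variety per placement.
; The indexed BuildingTextureData supplies diffuse, normal and specular textures to a dynamic material instance on slot 0.
; The event graph contains only empty event stubs.
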(event EventBeginPlay)

(event Collision|EventActorBeginOverlap
  (params (OtherActor :type Actor%20Object%20Reference)))

(event EventTick
  (params (DeltaSeconds :type Float%20%28single-precision%29)))

(fn CreateMIDFromTextureData
  (params (SwatchIndex :type Integer) (Mesh%20to%20Modify :type Primitive%20Component%20Object%20Reference))
  (bind _ar_btd (Variables|Default|GetAR_BTD))
  (bind _output (Utilities|Array|Get(acopy) _ar_btd SwatchIndex))
  (bind _returnvalue (Rendering|Material|CreateDynamicMaterialInstance Mesh%20to%20Modify 0 (Rendering|Material|GetMaterial Mesh%20to%20Modify 0) "None" :!class /Script/Engine.PrimitiveComponent))
  (Utilities|FlowControl|Sequence
    (:then_0
      (Rendering|Material|SetTextureParameterValue _returnvalue "Diffuse" (Variables|/Script/FortniteGame.BuildingTextureData|GetDiffuse _output)))
    (:then_1
      (Rendering|Material|SetTextureParameterValue _returnvalue "Normals" (Variables|/Script/FortniteGame.BuildingTextureData|GetNormal _output)))
    (:then_2
      (Rendering|Material|SetTextureParameterValue _returnvalue "SpecularMasks" (Variables|/Script/FortniteGame.BuildingTextureData|GetSpecular _output)))))

(fn RandomizeSwatchIndex
  (params (Swatch%20Transition%20Distance :type Float%20%28double-precision%29))
  (return (Math|Integer|%25%28Integer%29 (Math|Float|Round (/ (/ (Math|Float|float+float (Math|Float|Absolute(Float) (.x (Transformation|GetActorLocation))) (Math|Float|Absolute(Float) (.y (Transformation|GetActorLocation))) (Math|Float|Absolute(Float) (.z (Transformation|GetActorLocation)))) 3.0) Swatch%20Transition%20Distance)) (Utilities|Array|Length (Variables|Default|GetAR_BTD)))))

(fn ConstructionScript
  (|CallParentFunction)
  (if (Variables|Default|GetRandomizeTextureByPosition)
    (Utilities|FlowControl|Sequence
      (:then_0
        (bind _randomizedindex (CallFunction|RandomizeSwatchIndex 16.0))
        (Variables|Default|SetSwatchIndexA _randomizedindex))
      (:then_1
        _)))
  (CallFunction|CreateMIDFromTextureData (Variables|Default|GetSwatchIndexA) (Variables|Default|GetSM_Mystic_AncientMaskPaper_B)))
